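# Body material for the CosmoHound quadruped cosmetic companion, built on the shared 2023 character parent PreFX and PostFX functions.
# Between them, a CosmoHound Meteor custom-color function feeds an emissive and base-color override.
# A channel-selectable FX texture mask drives blends toward tinted emissive and black base color.
# Inverse Fresnel darkens glow edges, apparently for the eyes, before a SmoothStep remap; a brightness control adds a non-deferred mobile multiplier.
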Material /CosmeticCompanions/Assets/Quadruped/CosmoHound/Materials/M_CosmoHound_Body
graph {
 properties {
  bAutomaticallySetUsageInEditor=false
  bUseMaterialAttributes=true
  bUsedWithMeshDeformer=true
  bUsedWithMorphTargets=true
  bUsedWithSkeletalMesh=true
  bUsedWithStaticLighting=true
  previewMesh=/CosmeticCompanions/Assets/Quadruped/CosmoHound/Meshes/CosmoHound_Quadruped
  shadingModel=MSM_FromMaterialExpression
  usedShadingModels=Default Lit | Subsurface | Clear Coat | Subsurface Profile | Cloth
 }
 nodes [
  n1 = MaterialFunctionCall() {materialFunction=&MF_CharacterParent_2023_PreFX}
  n2 = MaterialFunctionCall!(Material=@n1.Result, Tech Art Mask=@n1.Result) {materialFunction=/CosmeticCompanions/Assets/Quadruped/CosmoHound_Meteor/Materials/MF_FN_CosmoHound_Custom_Color}
  n3 = GetMaterialAttributes(None=@n2.Base Color) {attributeGetTypes [#1, #2]}
  n4 = ScalarParameter() {parameterName=MinA}
  n5 = ScalarParameter() {defaultValue=1, parameterName=MaxA}
  n6 = Multiply(A=VectorParameter() {parameterName=EmissiveTint, defaultValue=<rgba 1 1 1 1>}.RGB, B=NamedRerouteUsage())
  n7 = ScalarParameter() {defaultValue=0.8, parameterName=EdgeOfEyeDarkeningFresenelExponent}
  n8 = Multiply(
   A=LinearInterpolate(A=@n3.Emissive Color, B=NamedRerouteDeclaration(None=@n6) {name=Tint, nodeColor=<rgba 1 0 0.5391 1>}, Alpha=NamedRerouteUsage())
   B=OneMinus(None=Fresnel(ExponentIn=@n7) {exponent=0.5})
  ) {constB=2}
  n9 = NamedRerouteUsage()
  n10 = TextureSampleParameter2D() {
   constMipValue=5
   parameterName=EmissiveMask
   samplerType=SAMPLERTYPE_Masks
   texture=/CosmeticCompanions/Assets/Quadruped/CosmoHound/Textures/T_CosmoHound_Body_FX
  }
  n11 = NamedRerouteDeclaration(None=ChannelMaskParameter(None=@n10.RGBA) {maskChannel=Alpha, parameterName=EmissiveMaskChannel, defaultValue=<rgba 0 0 0 1>}.R) {name=EmissiveMask, nodeColor=<rgba 1 0.9375 0 1>}
  MaterialFunctionCall(
   Material=SetMaterialAttributes!(
    MaterialAttributes=@n3.MaterialAttributes
    Emissive Color=SmoothStep(Min=@n4, Max=@n5, Value=@n8) {constMin=0.168}
    Base Color=LinearInterpolate(A=@n9, Alpha=@n11) {constB=0}
   ) {attributeSetTypes [#1, #2]}
   M=@n1.M
   Pre Skinned Local Position=@n1.M
   Distance Blend=@n1.M
   Fresnel=@n1.M
  ) {materialFunction=&MF_CharacterParent_2023_PostFX}
  MaterialFunctionCall!(
   Material=SetMaterialAttributes!(MaterialAttributes=@n2.Base Color, Base Color=Constant() {r=0.2}) {attributeSetTypes [#2]}
   Tech Art Mask
  ) {materialFunction=/CosmeticCompanions/Materials/MaterialFunctions/MF_FN_Companion_Custom_Color}
  NamedRerouteDeclaration(None=@n3.Base Color) {name=In_BaseColor, nodeColor=<rgba 0.125 1 0 1>}
  NamedRerouteDeclaration(None=GetMaterialAttributes() {attributeGetTypes [#1, #2]}.Base Color) {name=In_BaseColor_1, nodeColor=<rgba 0.125 1 0 1>}
  n12 = ScalarParameter() {defaultValue=1, group=06_Emissive, parameterName=EmissiveMobileMultiplier}
  n13 = Constant() {r=1}
  NamedRerouteDeclaration(None=Multiply(A=ScalarParameter() {defaultValue=1, group=06_Emissive, parameterName=EmissiveBrightness}, B=ShadingPathSwitch(Default=@n12, Deferred=@n13))) {name=Emissive Brightness Modulation, nodeColor=<rgba 1 0.3281 0 1>}
 ]
}
